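# Physics asset for the F_MED Mad Commander hat face accessory, providing secondary-motion physics for its hair and hood bones.
# Capsule bodies cover a front hair bone, two-segment left and right hair chains, and back, left and right hood bones, with two capsules shaping the head.
# Front hair, chain roots and hood bones constrain directly to the head, second hair segments to their first, all sharing narrow swing and tighter twist limits.
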PhysicsAsset /BRCosmetics/Accessories/Hats/F_MED_Mad_Commander_Hat/Meshes/F_MED_Mad_Commander_FaceAcc_Dyn
bodies {
 dyn_hair_01_l [{shapeType=Capsule, length=3.25, radius=0.75, center=<-2.3849 0 -0.0001>, rotation=<rot 90 19.4712 19.4713>}]
 dyn_hair_01_r [{shapeType=Capsule, length=3.25, radius=0.75, center=<-2.3849 0 -0.0001>, rotation=<rot 90 19.4712 19.4713>}]
 dyn_hair_02_l [{shapeType=Capsule, length=3.25, radius=0.75, center=<-2.3849 0 -0.0001>, rotation=<rot 90 19.4712 19.4713>}]
 dyn_hair_02_r [{shapeType=Capsule, length=3.25, radius=0.75, center=<-2.3849 0 -0.0001>, rotation=<rot 90 19.4712 19.4713>}]
 dyn_hair_fr [{shapeType=Capsule, length=3.25, radius=0.75, center=<-2.3849 0 -0.0001>, rotation=<rot 90 19.4712 19.4713>}]
 dyn_hood_bk [{shapeType=Capsule, length=2, radius=2, center=<-2.5882 0 0>, rotation=<rot 90 0.0148 0.0148>}]
 dyn_hood_l [{shapeType=Capsule, length=2, radius=2, center=<-2.5882 0 0>, rotation=<rot 90 0.0148 0.0148>}]
 dyn_hood_r [{shapeType=Capsule, length=2, radius=2, center=<2.5882 0 0>, rotation=<rot 90 0.0148 0.0148>}]
 head [
  {shapeType=Capsule, length=6, radius=4.25, center=<3.0032 4.7367 0.6166>, rotation=<rot 90 -0.0099 -0.0099>}
  {shapeType=Capsule, length=3.0118, radius=3.5883, center=<-1.932 7.1904 -2.453>, rotation=<rot 51.5259 90.4732 108.5849>}
 ]
}
constraints [
 {bone1Name=dyn_hair_02_r, bone2Name=dyn_hair_01_r, swing1LimitDegrees=20, swing2LimitDegrees=20, twistLimitDegrees=5}
 {bone1Name=dyn_hair_02_l, bone2Name=dyn_hair_01_l, swing1LimitDegrees=20, swing2LimitDegrees=20, twistLimitDegrees=5}
 {bone1Name=dyn_hood_l, bone2Name=head, swing1LimitDegrees=20, swing2LimitDegrees=20, twistLimitDegrees=5}
 {bone1Name=dyn_hair_fr, bone2Name=head, swing1LimitDegrees=20, swing2LimitDegrees=20, twistLimitDegrees=5}
 {bone1Name=dyn_hair_01_r, bone2Name=head, swing1LimitDegrees=20, swing2LimitDegrees=20, twistLimitDegrees=5}
 {bone1Name=dyn_hair_01_l, bone2Name=head, swing1LimitDegrees=20, swing2LimitDegrees=20, twistLimitDegrees=5}
 {bone1Name=dyn_hood_r, bone2Name=head, swing1LimitDegrees=20, swing2LimitDegrees=20, twistLimitDegrees=5}
 {bone1Name=dyn_hood_bk, bone2Name=head, swing1LimitDegrees=20, swing2LimitDegrees=20, twistLimitDegrees=5}
]
